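; Cluster bomblet projectile for the Commando's frag grenade perk, which explodes on a random timer or when damaged.
; BeginPlay waits a random delay, while any damage triggers a short delay; each path has its own DoOnce and runs an authority-only explosion at the actor's location.
; OnExploded spawns the particle and sound where cosmetic events can execute and fades the projectile audio.
; OnStop plays a bounce sound at the impact point.
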(event Game|Damage|EventAnyDamage
  (params (Damage :type Float%20%28single-precision%29) (DamageType :type Damage%20Type%20Object%20Reference) (InstigatedBy :type Controller%20Object%20Reference) (DamageCauser :type Actor%20Object%20Reference))
  ;[1 0.5515 0] Kills if takes damage
  (Utilities|FlowControl|DoOnce false
    (:Completed
      (Utilities|FlowControl|Delay 0.1)
      ;[1 0 0.01768] Explosion
      (Networking|SwitchHasAuthority
        (:Authority
          (Utilities|FlowControl|DoOnce false
            (:Completed
              (Variables|Default|SetExplosionLocation (Transformation|GetActorLocation))
              (Delivery|DoExplosion) ; Explosion
              (Audio|Components|Audio|FadeOut (Variables|Projectile_M80|GetAudio1) 0.2 0.0 "Linear" :!class /Script/Engine.AudioComponent))))) ; If touched or damagenly on server will it trigger explosion
    )))

(event EventBeginPlay
  (Utilities|FlowControl|Delay (Math|Random|RandomFloatinRange (Variables|Setup|GetMinExplosionDelay) (Variables|Setup|GetMaxExplosionDelay)) :!pinned true) ; Time Before Explosion
  ;[1 0 0.01768] Explosion
  (Networking|SwitchHasAuthority
    (:Authority
      (Utilities|FlowControl|DoOnce false
        (:Completed
          (Variables|Default|SetExplosionLocation (Transformation|GetActorLocation))
          (Delivery|DoExplosion) ; Explosion
          (Audio|Components|Audio|FadeOut (Variables|Projectile_M80|GetAudio1) 0.2 0.0 "Linear" :!class /Script/Engine.AudioComponent))))) ; If touched or damagenly on server will it trigger explosion
)

(event EventOnStop
  (params (Hit :type Hit%20Result%20Structure%20%28by%20ref%29))
  (bind (_bblockinghit _binitialoverlap _time _distance _location _impactpoint) (Collision|BreakHitResult Hit))
  (Audio|PlaySoundatLocation (Variables|Sounds|GetBounceSound) _impactpoint :!pinned true) ; Play Bounce Sound
)

(event EventOnExploded
  (params (HitActors :type Array%20of%20Actor%20Object%20References) (HitResults :type Array%20of%20Hit%20Result%20Structures))
  ;[1 1 1] OnExploded - Emitter & Sound
  (Networking|CanExecuteCosmeticEvents
    (:True
      (Effects|SpawnEmitteratLocation (Variables|Default|GetP_Explosion) (Variables|Default|GetExplosionLocation))
      (Audio|PlaySoundatLocation (Variables|Sounds|GetExplosionSound) (Variables|Default|GetExplosionLocation)) ; Play Explosion Sound
      (Audio|Components|Audio|FadeOut (Variables|Projectile_M80|GetAudio1) 0.2 0.0 "Linear" :!class /Script/Engine.AudioComponent))))

(fn ConstructionScript)
